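# Translucent UI material for the chainsaw reticle, combining a center dot with a mirrored chainsaw reticle shape that animates for an in-range state.
# A distance-field texture, sampled through a deprecated 2x1 symmetry function on normalized UI UVs, is shaped by an SDF glow/anti-aliasing function.
# AnimInRange interpolates the shape's horizontal offset and opacity between start and final parameters.
# Max nodes merge the result with the dot's plain and glow outputs.
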Material /ChainsawGameplay/UI/M_UI_Chainsaw_Reticle
graph {
 properties {bAutomaticallySetUsageInEditor=false, blendMode=BLEND_Translucent, materialDomain=MD_UI, usedShadingModels=Default Lit}
 nodes [
  n1 = MaterialFunctionCall!(Size=Constant() {r=0.02}) {materialFunction=/Game/Athena/HUD/reticles/MaterialFunctions/MF_Athena_Reticle_CenterDot}
  n2 = ScalarParameter() {defaultValue=0.1, group=InRange, parameterName=InRangeStartOffset}
  n3 = ScalarParameter() {defaultValue=-0.2, group=InRange, parameterName=InRangeFinalOffset}
  n4 = ScalarParameter() {defaultValue=1, group=InRange, parameterName=AnimInRange, sliderMax=1}
  n5 = MaterialFunctionCall!(
   UV=MaterialFunctionCall() {materialFunction=&GetUserInterfaceUV}.Normalized UV
   Offset=MaterialFunctionCall(X=LinearInterpolate(A=@n2, B=@n3, Alpha=@n4), Y=Constant()) {materialFunction=&MakeFloat2}.Result
   Rotation
   Scale=MaterialFunctionCall(X=Constant() {r=0.75}, Y=Constant() {r=0.375}) {materialFunction=&MakeFloat2}.Result
   MirrorY
  ) {materialFunction=/FortUILibrary/MaterialFunctions/Basics/MF_UI_Symmetry_2x1_DEPRECATED}
  n6 = Multiply(
   A=MaterialFunctionCall!(
    DistanceField=TextureSample(UVs=@n5.Result) {samplerType=SAMPLERTYPE_LinearGrayscale, texture=&/T_UI_Chainsaw_Reticle}.R
    GlowMin=Constant() {r=0.7}
    GlowMax=Constant() {r=0.6}
   ) {materialFunction=/FortUILibrary/MaterialFunctions/Shapes/Modifiers/MF_UI_SDFGlowOrAA}.Result
   B=LinearInterpolate(
    A=ScalarParameter() {defaultValue=0.7, group=InRange, parameterName=InRangeStartOpacity}
    B=ScalarParameter() {defaultValue=1, group=InRange, parameterName=InRangeFinalOpacity}
    Alpha=ScalarParameter() {defaultValue=1, group=InRange, parameterName=AnimInRange, sliderMax=1}
   )
  )
  Max(A=@n1.Dot, B=@n6)
  Max(A=@n1.DotWithGlow, B=@n6)
 ]
}
