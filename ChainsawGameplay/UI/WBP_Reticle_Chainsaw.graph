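# Chainsaw reticle extension widget with a center reticle, a fuel meter and a "Hold to Travel Further" keybind prompt.
# An async Overheat attribute listener feeds the GasMeter material's Progress as inverted, normalized fuel.
# The DidntHoldScooterButton gameplay event shows the prompt through a DoOnce, which a looping hide timer resets; debug PrintStrings remain.
# Cooldown changes toggle the center image, and Anim_InRange, never played here, animates its material.
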WidgetBlueprint /ChainsawGameplay/UI/WBP_Reticle_Chainsaw
animations [Anim_InRange {
  binding_count=1
  length_seconds=0.35
  object=&self:Anim_InRange
  sequence {bindings [CenterImage {
     id=#1
     possessed_class=Image
     tracks [{
       type=MovieSceneWidgetMaterialTrack
       sections [{type=MovieSceneParameterSection, channels [AnimInRange {type=MovieSceneScriptingFloatChannel, keys [[0,0],[7,1]]}]}]
      }]
    }]}
 }]
graphs [
 EventGraph {nodes [
   n1 = HideHoldText() {is=title K2Node_CustomEvent}
   > SetVisibility(
    self=[
     Get HoldPrompt_Keybind() -> (HoldPrompt_Keybind:Keybind Widget Object Reference)
     Get HoldPrompt_Text1() -> (HoldPrompt_Text1:Common Text Object Reference)
     Get HoldPrompt_Text2() -> (HoldPrompt_Text2:Common Text Object Reference)
    ]
    InVisibility=Hidden
   )
   > n2 = PrintString!(execute, InString="HIDE!", bPrintToScreen, bPrintToLog, TextColor, Duration=20, Key) -> (then)
   Do Once(execute=@n7.EventReceived, Reset=@n2.then)
   > Clear and Invalidate Timer by Handle(Handle=Get HoldVisibilityTimer() -> (HoldVisibilityTimer:Timer Handle Structure))
   > SetVisibility(self=[Get HoldPrompt_Text2(), Get HoldPrompt_Keybind(), Get HoldPrompt_Text1()], InVisibility=Visible)
   > PrintString(InString="SHOW!", Duration=20)
   > n3 = Set Timer by Event(Delegate=@n1, Time=2, bLooping=true)
   > Set HoldVisibilityTimer(HoldVisibilityTimer:Timer Handle Structure=@n3) -> (Output_Get:Timer Handle Structure)
   n4 = Cast To FortPlayerController(Object:Wildcard)
   > n5 = Bind Event to On View Target Changed(self:Fort Player Controller Object Reference=@n4) {is=title K2Node_AddDelegate}
   n6 = WaitForAttributeChanged!(
    execute:Exec=[@n5.then, @n8.then]
    TargetActor:Actor Object Reference=GetViewTarget!(self:Controller Object Reference=GetOwningPlayer()) -> (ReturnValue:Actor Object Reference)
    Attribute:Gameplay Attribute Structure=Get OverheatAttribute() -> (OverheatAttribute:Gameplay Attribute Structure)
    OnlyTriggerOnce:bool=false
   ) -> (
    then:Exec
    AsyncTaskProxy:Ability Async Wait Attribute Changed Object Reference
    Changed:Exec
    Attribute:Gameplay Attribute Structure
    NewValue:float=~
    OldValue:float=~
   ) {is=title K2Node_AsyncAction}
   > Set OverheatListener(OverheatListener:Ability Async Wait Attribute Changed Object Reference=@n6) -> (Output_Get:Ability Async Wait Attribute Changed Object Reference)
   > UpdateFuelBar!(
    execute
    self
    NewValue:double=GetFloatAttribute!(Actor:Actor Object Reference=GetOwningPlayer(), Attribute:Gameplay Attribute Structure=Get OverheatAttribute()) -> (bSuccessfullyFoundAttribute:bool, ReturnValue:float).ReturnValue
   ) -> (then)
   > SetVisibility(self=[Get HoldPrompt_Keybind(), Get HoldPrompt_Text1(), Get HoldPrompt_Text2()], InVisibility=Hidden)
   > n7 = WaitGameplayEventToActor!(
    execute:Exec
    TargetActor:Actor Object Reference=GetOwningPlayer()
    EventTag:Gameplay Tag Structure=(TagName="Event.Athena.Chainsaw.DidntHoldScooterButton")
    OnlyTriggerOnce:bool=false
    OnlyMatchExact:bool=true
   ) -> (then:Exec, AsyncTaskProxy:Ability Async Wait Gameplay Event Object Reference, EventReceived:Exec, Payload:Gameplay Event Data Structure) {is=title K2Node_AsyncAction}
   UpdateFuelBar(execute=@n6.Changed, NewValue=@n6.NewValue)
   Event Destruct()
   > Parent: Destruct()
   > Cancel(self:Cancellable Async Action Object Reference=Get OverheatListener() -> (OverheatListener:Ability Async Wait Attribute Changed Object Reference))
   Event Construct()
   > n8 = Parent: Construct()
  ]}
 OnCooldownStateChanged {nodes [
   n1 = OnCooldownStateChanged!() -> (then, IsCooldownActive:bool=~) {is=title K2Node_FunctionEntry}
   > Set Is Cooldown Active(Is Cooldown Active:bool=@n1) -> (Output_Get:bool)
   > RefreshReticleVisibility()
  ]}
 RefreshReticleVisibility {nodes [
   RefreshReticleVisibility() {is=title K2Node_FunctionEntry}
   > SetVisibility(
    self=Get CenterImage() -> (CenterImage:Image Object Reference)
    InVisibility=Select(Option 0=Visible, Option 1=Hidden, Index=Get Is Cooldown Active() -> (Is Cooldown Active:bool))
   )
  ]}
 UpdateFuelBar {nodes [
   n1 = UpdateFuelBar!() -> (then, NewValue:double=~) {is=title K2Node_FunctionEntry}
   > n2 = GetDynamicMaterial(self=Get GasMeter() -> (GasMeter:Image Object Reference))
   > SetScalarParameterValue(
    self=@n2
    ParameterName=Progress
    Value=float - float(A=1, B=Remap Float(Input=@n1, OldRange=(X=0,Y=100), NewRange=(X=0,Y=1), Clamp=true) -> (Output=~))
   )
  ]}
]
variables [
 HoldVisibilityTimer {type=TimerHandle}
 Is Cooldown Active {type=boolean}
 OverheatAttribute {
  type=GameplayAttribute
  default {
   attribute=/Script/ChainsawGameplayRuntime.ChainsawAttributeSet:Overheat
   attributeName=Overheat
   attributeOwner=&ChainsawAttributeSet
  }
 }
 OverheatListener {type=AbilityAsync_WaitAttributeChanged, default=null}
]
widgets {
 info {
  inheritedWidgetCount=1
  parentClass=/Script/FortniteUI.FortWeaponReticleExtensionWidgetBase
  rootWidgetClass=&Overlay
  widgetCount=9
 }
 widgets [
  Reticle_Root {widgetClassPath=&Overlay, widget=Reticle_Root}
  CenterImage {
   widgetClassPath=&Image
   bIsVariable=true
   parent=Reticle_Root
   widget=CenterImage
   properties {
    brush {
     bIsDynamicallyLoaded=false
     drawAs=Image
     imageType=FullColor
     mirroring=NoMirror
     resourceName=None
     resourceObject=&/MI_UI_Chainsaw_Reticle
     tiling=NoTile
     imageSize=<172 172>
     margin=<ltrb 0 0 0 0>
     outlineSettings {
      bUseBrushTransparency=false
      roundingType=HalfHeightRadius
      width=0
      color {colorUseRule=UseColor_Specified, specifiedColor=<rgba 0 0 0 0>}
      cornerRadii=<0 0 0 0>
     }
     tintColor {colorUseRule=UseColor_Specified, specifiedColor=<rgba 1 1 1 1>}
     uVRegion {bIsValid=false, max=<0 0>, min=<0 0>}
    }
    colorAndOpacity=<rgba 1 1 1 0.9>
   }
   slot {class=OverlaySlot, properties {horizontalAlignment=HAlign_Center, verticalAlignment=VAlign_Center}}
  }
  Overlay_TempReorg {
   widgetClassPath=&Overlay
   parent=Reticle_Root
   widget=Overlay_TempReorg
   properties {renderTransform {angle=0, scale=<1 1>, shear=<0 0>, translation=<65 20>}}
   slot {class=OverlaySlot}
  }
  GasMeter {
   widgetClassPath=&Image
   bIsVariable=true
   parent=Overlay_TempReorg
   widget=GasMeter
   properties {brush {
     bIsDynamicallyLoaded=false
     drawAs=Image
     imageType=FullColor
     mirroring=NoMirror
     resourceName=None
     resourceObject=&/MI_UI_Chainsaw_Fuel
     tiling=NoTile
     imageSize=<256 256>
     margin=<ltrb 0 0 0 0>
     outlineSettings {
      bUseBrushTransparency=false
      roundingType=HalfHeightRadius
      width=0
      color {colorUseRule=UseColor_Specified, specifiedColor=<rgba 0 0 0 0>}
      cornerRadii=<0 0 0 0>
     }
     tintColor {colorUseRule=UseColor_Specified, specifiedColor=<rgba 1 1 1 1>}
     uVRegion {bIsValid=false, max=<0 0>, min=<0 0>}
    }}
   slot {class=OverlaySlot}
  }
  HoldPrompt {
   widgetClassPath=&HorizontalBox
   parent=Overlay_TempReorg
   widget=HoldPrompt
   properties {renderTransform {angle=0, scale=<1 1>, shear=<0 0>, translation=<200 59>}}
   slot {class=OverlaySlot, properties {horizontalAlignment=HAlign_Right, verticalAlignment=VAlign_Center}}
  }
  HoldPrompt_Text1 {
   widgetClassPath=&CommonTextBlock
   bIsVariable=true
   parent=HoldPrompt
   widget=HoldPrompt_Text1
   properties {
    style=/FortUILibrary/Styles/Typography/_TextStyle_TagNormal._TextStyle_TagNormal_C
    text="Hold "
    visibility=Hidden
    font {
     bForceMonospaced=false
     bMaterialIsStencil=false
     fontMaterial=null
     fontObject=&BurbankSmall
     letterSpacing=0
     monospacedWidth=1
     size=15
     skewAmount=0
     typefaceFontName=Black
     outlineSettings {
      bApplyOutlineToDropShadows=false
      bMiteredCorners=false
      bSeparateFillAlpha=false
      outlineMaterial=null
      outlineSize=0
      outlineColor=<rgba 0 0 0 1>
     }
    }
    shadowOffset=<0 0>
    strikeBrush {
     bIsDynamicallyLoaded=false
     drawAs=NoDrawType
     imageType=NoImage
     mirroring=NoMirror
     resourceName=None
     resourceObject=null
     tiling=NoTile
     imageSize=<32 32>
     margin=<ltrb 0 0 0 0>
     outlineSettings {
      bUseBrushTransparency=false
      roundingType=HalfHeightRadius
      width=0
      color {colorUseRule=UseColor_Specified, specifiedColor=<rgba 0 0 0 0>}
      cornerRadii=<0 0 0 0>
     }
     tintColor {colorUseRule=UseColor_Specified, specifiedColor=<rgba 1 1 1 1>}
     uVRegion {bIsValid=false, max=<0 0>, min=<0 0>}
    }
   }
   slot {class=HorizontalBoxSlot, properties {verticalAlignment=VAlign_Center, padding=<ltrb 4 0 0 0>}}
  }
  HoldPrompt_Keybind {
   widgetClassPath=/FortUILibrary/Input/KeybindWidget
   bIsVariable=true
   parent=HoldPrompt
   widget=HoldPrompt_Keybind
   properties {boundAction=Target, visibility=Hidden}
   slot {class=HorizontalBoxSlot}
  }
  HoldPrompt_Text2 {
   widgetClassPath=&CommonTextBlock
   bIsVariable=true
   parent=HoldPrompt
   widget=HoldPrompt_Text2
   properties {
    style=&_TextStyle_TagNormal_C
    text=to Travel Further
    visibility=Hidden
    font {
     bForceMonospaced=false
     bMaterialIsStencil=false
     fontMaterial=null
     fontObject=&BurbankSmall
     letterSpacing=0
     monospacedWidth=1
     size=15
     skewAmount=0
     typefaceFontName=Black
     outlineSettings {
      bApplyOutlineToDropShadows=false
      bMiteredCorners=false
      bSeparateFillAlpha=false
      outlineMaterial=null
      outlineSize=0
      outlineColor=<rgba 0 0 0 1>
     }
    }
    shadowOffset=<0 0>
    strikeBrush {
     bIsDynamicallyLoaded=false
     drawAs=NoDrawType
     imageType=NoImage
     mirroring=NoMirror
     resourceName=None
     resourceObject=null
     tiling=NoTile
     imageSize=<32 32>
     margin=<ltrb 0 0 0 0>
     outlineSettings {
      bUseBrushTransparency=false
      roundingType=HalfHeightRadius
      width=0
      color {colorUseRule=UseColor_Specified, specifiedColor=<rgba 0 0 0 0>}
      cornerRadii=<0 0 0 0>
     }
     tintColor {colorUseRule=UseColor_Specified, specifiedColor=<rgba 1 1 1 1>}
     uVRegion {bIsValid=false, max=<0 0>, min=<0 0>}
    }
   }
   slot {class=HorizontalBoxSlot, properties {verticalAlignment=VAlign_Center, padding=<ltrb 4 0 0 0>}}
  }
  TutorialHighlight {widgetClassPath=&TutorialHighlight}
 ]
}
